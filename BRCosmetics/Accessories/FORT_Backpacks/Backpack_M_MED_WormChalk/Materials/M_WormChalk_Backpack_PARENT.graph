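# Parent material for the WormChalk backpack cosmetic: a masked, two-sided Cloth-shaded surface built on the 2023 anime character shader functions.
# The shader's Pre, Calculations and Post functions run in sequence to produce the material attributes.
# Two-sided sign gives back faces one flat color parameter, a constant threshold gradient and no DFL mask.
# A selectable FX channel, power-adjusted and multiplied by its temporal dither, sets the opacity mask.
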Material /BRCosmetics/Accessories/FORT_Backpacks/Backpack_M_MED_WormChalk/Materials/M_WormChalk_Backpack_PARENT
graph {
 properties {
  bAutomaticallySetUsageInEditor=false
  bUseMaterialAttributes=true
  bUsedWithMeshDeformer=true
  bUsedWithMorphTargets=true
  bUsedWithSkeletalMesh=true
  bUsedWithStaticLighting=true
  blendMode=BLEND_Masked
  pixelDepthOffsetMode=PDOM_Legacy
  shadingModel=MSM_Cloth
  twoSided=true
  usedShadingModels=Cloth
 }
 nodes [
  n1 = MaterialFunctionCall!(CustomUV) {materialFunction=/Game/Global/Materials/Character_Parent_2023/Anime_Parent_2023/MaterialFunctions/MFC_AnimeShader_2023_Pre}
  n2 = MaterialFunctionCall!(CustomLightVector, OverrideNormal_WS, STT=@n1.STT, Normal_Map=@n1.STT, DFL=@n1.STT, FX=@n1.STT) {materialFunction=/Game/Global/Materials/Character_Parent_2023/Anime_Parent_2023/MaterialFunctions/MFC_AnimeShader_2023_Calculations}
  n3 = Saturate(None=TwoSidedSign())
  n4 = VectorParameter() {group=WormChalk, parameterName=ColorLit, defaultValue=<rgba 0.8716 0.8509 0.8879 1>}
  n5 = Power(
   Base=ChannelMaskParameter(None=@n1.FX) {group=WormChalk, parameterName=OpacityMaskChannel}.R
   Exp=ScalarParameter() {defaultValue=1, group=WormChalk, parameterName=OpacityMaskPow}
  ) {constExponent=1}
  SetMaterialAttributes!(
   MaterialAttributes=MaterialFunctionCall!(
    Shading_Gradient=@n2.Shading_Gradient
    Threshold_Gradient=LinearInterpolate(B=@n2.Threshold_Gradient, Alpha=@n3) {constA=1}
    Specular_Gradient=@n2.Shading_Gradient
    Sheen_Mask=@n2.Shading_Gradient
    Thickness=@n2.Shading_Gradient
    DFL_Mask=LinearInterpolate(B=@n2.DFL_Mask, Alpha=@n3)
    FX=@n2.Shading_Gradient
    Metallic_Gradient=@n2.Shading_Gradient
    RimLight_Gradient=@n2.Shading_Gradient
    NormalBuffer=@n2.Shading_Gradient
    --------
    CINE_LightSources=@n2.Shading_Gradient
    CINE_LightColor=@n2.Shading_Gradient
    ---------
    Color_Lit=LinearInterpolate(A=@n4.RGB, B=@n1.Color_Lit, Alpha=@n3)
    Color_Shaded=LinearInterpolate(A=@n4.RGB, B=@n1.Color_Shaded, Alpha=@n3)
    SRM=@n1.SRM
    Emissive=@n1.SRM
    DFL_Color=@n1.SRM
    Specular_Color=@n1.SRM
    RimLight_Color=@n1.SRM
   ) {materialFunction=/Game/Global/Materials/Character_Parent_2023/Anime_Parent_2023/MaterialFunctions/MFC_AnimeShader_2023_Post}.Result
   Opacity Mask=Multiply(
    A=@n5
    B=MaterialFunctionCall(Alpha Threshold=@n5) {materialFunction=&DitherTemporalAA}.Result
   )
   Opacity
   Emissive Color
  ) {attributeSetTypes [#1, #2, #3]}
 ]
}
