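# Animation blueprint for the WormChalk backpack that plays its idle animation and adjusts the pack bone to fit the wearer.
# Three additive Modify Bone nodes apply a body-dependent offset, a spine-driven shift and roll, and a fixed offset.
# Each update maps the owner's Spine_05-to-Spine_01 yaw delta from ref pose to the second node's alpha.
# Initialization and customization-loaded events pick the body offset by F_MED_Prime mesh, otherwise by gender.
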AnimBlueprint /BRCosmetics/Accessories/FORT_Backpacks/Backpack_M_MED_WormChalk/Meshes/M_MED_WormChalk_Pack_AnimBP
target_skeleton=&/M_MED_WormChalk_Pack_Skeleton
graphs [
 AnimGraph {nodes [
   Input Pose()
   n1 = Sequence Player 'WormChalk_Idle_PACK'() {sequence=/BRCosmetics/Accessories/FORT_Backpacks/Backpack_M_MED_WormChalk/Animation/WormChalk_Idle_PACK}
   n2 = "Transform (Modify) Bone - Bone: pack"(
    ComponentPose="Transform (Modify) Bone - Bone: pack"(ComponentPose=Local To Component(LocalPose=@n1), Translation=Get Begin_Offset() -> (Begin_Offset:Vector), Rotation="0,0,0") {translationMode=BMM_Additive, alphaBoolBlend {blendOption=Linear}, boneToModify {boneName=pack}}
    Alpha=Get Spine_Rotate_Alpha() -> (Spine_Rotate_Alpha:double)
    Translation="0,-2,1.8"
    Rotation="0,0,13"
   ) {
    rotationMode=BMM_Additive
    translationMode=BMM_Additive
    alphaBoolBlend {blendOption=Linear}
    boneToModify {boneName=pack}
    rotation {roll=13}
    translation {y=-2, z=1.8}
   }
   n3 = "Transform (Modify) Bone - Bone: pack"(ComponentPose=@n2, Translation="0,-4,0", Rotation="0,0,0") {translationMode=BMM_Additive, alphaBoolBlend {blendOption=Linear}, boneToModify {boneName=pack}, translation {y=-4}}
   Output Pose(Result=Component To Local(ComponentPose=@n3)) {layerGroup=None}
  ]}
 EventGraph {nodes [
   Event BlueprintUpdateAnimation()
   n1 = Get Character_SkelMesh() -> (Character_SkelMesh:Skeletal Mesh Component Object Reference)
   > Set Spine_Rotate_Alpha(Spine_Rotate_Alpha:double=MapRangeClamped(
     Value=GetDeltaTransformFromRefPose!(self=@n1, BoneName=Spine_05, BaseName=Spine_01) -> (
      ReturnValue_Location:Vector
      ReturnValue_Rotation_Roll:float
      ReturnValue_Rotation_Pitch:float
      ReturnValue_Rotation_Yaw:float
      ReturnValue_Scale:Vector
     ).ReturnValue_Rotation_Yaw
     InRangeA=-12
     InRangeB=-60
    )) -> (Output_Get:double)
   Event BlueprintBeginPlay() {is=title K2Node_Event}
   > Set Character_SkelMesh(Character_SkelMesh:Skeletal Mesh Component Object Reference=GetOwnerSkeletalMeshComponent()) -> (Output_Get:Skeletal Mesh Component Object Reference)
   n2 = Event OnCharacterCustomizationLoaded()
   n3 = Get Array Item with EFort Custom Part Type(Array:Array of Custom Character Part Object References=@n2.CustomParts) -> (Output:Custom Character Part Object Reference)
   n4 = Branch(
    execute=[@n2.then, @n6.then]
    Condition="Equal (Object)"(
     A=GetSkeletalMesh!(self:Custom Character Part Object Reference=@n3) -> (ReturnValue:Skeletal Mesh Object Reference)
     B=/DefaultCosmeticsAthena/Characters/Meshes/F_MED_Prime
    )
   )
   > Set Begin_Offset(Begin_Offset:Vector="0,0.4,0") -> (Output_Get:Vector)
   n5 = Branch(
    execute=@n4.else
    Condition="Equal (Enum)"(
     A:EFortCustomGender Enum=Get GenderPermitted!(self:Custom Character Part Object Reference=@n3) -> (GenderPermitted:EFortCustomGender Enum=Invalid)
     B:EFortCustomGender Enum=Female
    )
   )
   > Set Begin_Offset(Begin_Offset="0,-1.5,0")
   Set Begin_Offset(execute=@n5.else, Begin_Offset="0,-1.9,0")
   n6 = Event BlueprintInitializeAnimation()
  ]}
]
variables [
 Begin_Offset {type=Vector}
 Character_SkelMesh {type=SkeletalMeshComponent}
 Spine_Rotate_Alpha {type=number}
]
